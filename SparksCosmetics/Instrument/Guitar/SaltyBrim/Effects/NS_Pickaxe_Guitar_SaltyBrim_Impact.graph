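# Niagara impact effect for the SaltyBrim guitar pickaxe cosmetic: a short, blue-tinted burst with flashes, sparks, smoke and debris.
# One-shot burst emitters cover the core impact, layered flashes, smoke, sparks, directional spikes and surface-dependent debris, often using spawn groups to pack several looks into one emitter.
# System values for global scale and time-of-day color feed several emitters, and front-end modifiers adjust sizes, speeds and brightness for menu previews.
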NiagaraSystem /SparksCosmetics/Instrument/Guitar/SaltyBrim/Effects/NS_Pickaxe_Guitar_SaltyBrim_Impact
dependencies {
 usedDataInterfaces [
  &NiagaraDataInterfaceCurve
  &NiagaraDataInterfaceArrayFloat
  &NiagaraDataInterfaceColorCurve
  &NiagaraDataInterfaceMeshRendererInfo
  &NiagaraDataInterfaceVector2DCurve
 ]
 usedDynamicInputs []
 usedModules [
  &self:SystemSpawnScript.SystemScriptSource.SystemScriptGraph.NiagaraNodeAssignment_1.SetVariables_#1
  &SystemState
 ]
}
emitters {
 Core_Impact_Burst001 {
  emitterUpdateScript [
   &EmitterState()
   &SpawnBurst_Instantaneous(
    Spawn Probability=&FN_FrontEndFloatModifier(Gameplay=1, Front End=1)
    Use Spawn Probability=true
    UseLoopCountLimit=false
   ) {module=SpawnBurst_Instantaneous001}
   &SpawnBurst_Instantaneous(Spawn Group=1, Use Spawn Probability=false, UseLoopCountLimit=false)
   &SpawnBurst_Instantaneous(Spawn Group=2, Use Spawn Probability=false, UseLoopCountLimit=false) {module=SpawnBurst_Instantaneous002}
   &SpawnBurst_Instantaneous(
    Spawn Count=10
    Spawn Group=3
    Spawn Probability=&FN_FrontEndFloatModifier(Gameplay=1, Front End=1)
    Use Spawn Probability=true
   ) {module=SpawnBurst_Instantaneous003}
  ]
  particleSpawnScript [
   &PartitionParticles(Number Of Partitions=3)
   /SetVariables(Particles.VisibilityTag=&SpawnGroup()) {isSetParameters=true}
   &InitializeParticle(
    Lifetime=&MaskFloatBySpawnGroup(
     Masked Float=&FN_FrontEndFloatModifier(Gameplay=0.22, Front End=0.2)
     Passthrough Float=&MaskFloatBySpawnGroup(
      Masked Float=0.35
      Passthrough Float=&MaskFloat(
       Mask Source=0
       Mask Parameter=?
       Mask=3
       Masked Float=&RandomRangeFloat(Minimum=0.18, Maximum=0.2)
       Passthrough Float=1
      )
      Spawn Group Mask=2
     )
     Spawn Group Mask=1
    )
    Mesh Scale Mode=::NewEnumerator3
    Mesh Uniform Scale=0.35
    Sprite Rotation Mode=::NewEnumerator2
    Sprite Size=&MaskVector2dBySpawnGroup(
     Masked Vector2D=&UniformRangedVector2D(Minimum=<5 25>, Maximum=<8 55>)
     Passthrough Vector2D=&Vector2DFromFloat(Value=&Multiply_Float(A=155, B=$System.GlobalScale))
     Spawn Group Mask=3
    )
    Sprite Size Mode=::NewEnumerator4
    UsePositionOffset=false
    Write Material Random=true
   )
   &FN_PerParticleLOD(Enable SpriteSize Distance Scaling=true, Sprite Size Scale Far=<2.5 2.5>)
   &ShapeLocation(
    MaskSpawnGroup=false
    Rotation Mode=::NewEnumerator2
    Sphere Distribution=::NewEnumerator3
    Sphere Radius=&SelectFloatFromArray(
     Direct Array Index=&SpawnGroup()
     Float Selection Array=DataInterface() {propertyValues {FloatData [0,0,0,0.5,0]}}
    )
    UseOwnerScale=true
    Yaw / Pitch / Roll=&RandomRangeVector(Minimum=<0 0 0>, Maximum=<1 1 1>)
   )
   &AddVelocityFromPoint(
    Module.Use Falloff Distance=false
    Velocity Strength=&MaskFloatBySpawnGroup(
     Masked Float=&FloatFromCurve(
      FloatCurve=DataInterface() {propertyValues {Curve {keys [<at 0.0101 172.8235>,<at 0.1521 116.2374>,<at 0.375 80.9199>,<at 1 20>]}}}
      CurveIndex=&RandomRangeFloat(Minimum=0, Maximum=1)
      Scale Curve=$System.GlobalScale
     )
     Passthrough Float=&MaskFloatBySpawnGroup(
      Masked Float=&FloatFromCurve(
       FloatCurve=DataInterface() {propertyValues {Curve {keys [<at 0 0>,<at 1 15>]}}}
       CurveIndex=&RandomRangeFloat(Minimum=0, Maximum=1)
       Scale Curve=$System.GlobalScale
      )
      Passthrough Float=0
      Spawn Group Mask=4
     )
     Spawn Group Mask=3
    )
   )
   &InitialMeshOrientation(Orientation Axis=<0 0 1>, Orientation Vector=<0 0 1>, Rotation=<0 0.25 0>, Use Orientation Vector=true, Use Rotation Vector=true)
   &InitialMeshOrientation(
    Orientation Axis=<0 0 0>
    Orientation Vector=<0 0 0>
    Rotation=&RandomRangeVector(Minimum=<0 0 0>, Maximum=<1 0 0>)
    Use Orientation Vector=false
    Use Rotation Vector=true
   ) {module=InitialMeshOrientation001}
   /SetVariables(
    Particles.Flash Star Cam Offset=15
    Particles.Flash Star Lifetime=0.18
    Particles.Flash Star Rotation:NiagaraFloat=&RandomRangeFloat(Minimum=0, Maximum=360)
    Particles.Flash Star Sprite Size:Vector2f=&Multiply_Vector2D_ByFloat(Float=&Multiply_Float(A=$System.GlobalScale, B=65), Vector2D=<1 1>)
    Particles.Flash Star SubUV Index:NiagaraFloat=&RandomRangeFloat(Minimum=0, Maximum=3)
   ) {isSetParameters=true}
   /SetVariables(
    Particles.FlareCardColor:LinearColor=$System.TOD Color
    Particles.FlareCardDynamicParameter:Vector4f=&MakeVector4(X=0.2, Y=1, Z=0, W=0)
   ) {isSetParameters=true}
   &CameraOffset(Camera Offset Amount=&MaskFloatBySpawnGroup(Masked Float=0, Passthrough Float=20, Spawn Group Mask=3))
   &ApplyInitialForces()
   /SetVariables(Particles.DarkGlowDyn0:Vector4f=&MakeVector4(X=1, Y=1, Z=1, W=1)) {isSetParameters=true}
  ]
  particleUpdateScript [
   &ParticleState()
   &ScaleSpriteSize(
    Scale Factor=&MaskVector2dBySpawnGroup(
     Masked Vector2D=&Multiply_Vector2D_ByFloat(
      Float=&FloatFromCurve(FloatCurve=DataInterface() {propertyValues {Curve {
          defaultValue=1
          keys [
           {arriveTangent=5.1534, interpMode=RCIM_Cubic, leaveTangent=5.1534, tangentMode=RCTM_User, time=-0.0016, value=0.4676}
           {interpMode=RCIM_Cubic, tangentMode=RCTM_User, time=0.2, value=1}
           {arriveTangent=-2.876, interpMode=RCIM_Cubic, leaveTangent=-2.876, tangentMode=RCTM_User, time=1, value=0}
          ]
         }}})
      Vector2D=<1 1>
     )
     Passthrough Vector2D=&MaskVector2dBySpawnGroup(
      Masked Vector2D=&MakeVector2D(
       X=&FloatFromCurve(
        FloatCurve=DataInterface() {propertyValues {Curve {
           defaultValue=1
           keys [
            {arriveTangent=6.211, interpMode=RCIM_Cubic, leaveTangent=6.211, tangentMode=RCTM_User, time=0, value=0.5}
            {arriveTangentWeight=0.1005, interpMode=RCIM_Cubic, leaveTangentWeight=0.2345, tangentMode=RCTM_User, time=0.2, value=1}
            {arriveTangentWeight=0.6964, interpMode=RCIM_Cubic, tangentMode=RCTM_User, time=1.0035, value=0.3}
           ]
          }}}
        Scale Curve=$System.GlobalScale
       )
       Y=&FloatFromCurve(
        FloatCurve=DataInterface() {propertyValues {Curve {
           defaultValue=1
           keys [
            {interpMode=RCIM_Cubic, leaveTangentWeight=0.34, tangentMode=RCTM_User, time=0, value=0.8}
            {arriveTangent=0.6525, interpMode=RCIM_Cubic, leaveTangent=0.6525, tangentMode=RCTM_User, time=1, value=1.2}
           ]
          }}}
        Scale Curve=$System.GlobalScale
       )
      )
      Passthrough Vector2D=&Vector2DFromFloat(Value=&FloatFromCurve(FloatCurve=DataInterface() {propertyValues {Curve {
           defaultValue=1
           keys [
            {arriveTangent=1.3623, interpMode=RCIM_Cubic, leaveTangent=1.3623, tangentMode=RCTM_User, time=0, value=0.5}
            {arriveTangent=0.2407, arriveTangentWeight=0.7893, interpMode=RCIM_Cubic, leaveTangent=0.2407, tangentMode=RCTM_User, time=0.993, value=1}
           ]
          }}}))
      Spawn Group Mask=3
     )
     Spawn Group Mask=1
    )
    Scale Sprite Size Mode=::NewEnumerator2
   )
   &LinearForce(Force=&Multiply_VectorByFloat(
     Vector=$System.DirectionVector
     Float=&MaskFloatBySpawnGroup(
      Masked Float=&SelectFloatFromArray(
       Direct Array Index=$Particles.PartitionParticles.Partition
       Float Selection Array=DataInterface() {propertyValues {FloatData [25,3900,4500]}}
      )
      Passthrough Float=0
      Spawn Group Mask=4
     )
    ))
   &ScaleMeshSize(Scale Factor=&VectorFromFloat(Value=&FloatFromCurve(
      FloatCurve=DataInterface() {propertyValues {Curve {keys [
          {arriveTangent=7.2985, interpMode=RCIM_Cubic, leaveTangent=7.2985, tangentMode=RCTM_User, time=0, value=0.5}
          {arriveTangent=0.6472, interpMode=RCIM_Cubic, leaveTangent=0.6472, tangentMode=RCTM_User, time=0.15, value=1.25}
          {arriveTangent=1.8898, interpMode=RCIM_Cubic, leaveTangent=1.8898, tangentMode=RCTM_User, time=1, value=2}
         ]}}}
      Scale Curve=$System.GlobalScale
     )))
   &ScaleColor(
    Scale Alpha=&FloatFromCurve(FloatCurve=DataInterface() {propertyValues {Curve {
        defaultValue=1
        keys [
         {arriveTangent=-0.2961, interpMode=RCIM_Cubic, leaveTangent=-0.2961, tangentMode=RCTM_User, time=0.1, value=1}
         {arriveTangentWeight=0.4395, interpMode=RCIM_Cubic, tangentMode=RCTM_User, time=1, value=0}
        ]
       }}})
    Scale RGB=&MakeVectorFromLinearColorRGB(Color=&ColorFromCurve(ColorCurve=DataInterface() {propertyValues {
        AlphaCurve {keys [<at 0.0782 1>,<at 1 0>]}
        BlueCurve {keys [<at 0 6.1387>,<at 0.3095 4>,<at 0.9275 0.1849>]}
        GreenCurve {keys [<at 0 7>,<at 0.3095 0.4886>,<at 0.9275 0.0366>]}
        RedCurve {keys [<at 0 0.7109>,<at 0.3095 0>,<at 0.9275 0>]}
       }}))
    ScaleA=true
    ScaleRGB=true
   )
   /SetVariables(Particles.Update Flash Star Age:NiagaraFloat=&Divide_Float(A=$Particles.Age, B=$Particles.Initial.Flash Star Lifetime)) {isSetParameters=true}
   /SetVariables(
    Particles.Update Flash Star Sprize Size:Vector2f=&Multiply_Vector2D_ByFloat(
     Float=&FloatFromCurve(
      FloatCurve=DataInterface() {propertyValues {Curve {
         defaultValue=0.2
         keys [
          {arriveTangent=1.826, interpMode=RCIM_Cubic, leaveTangent=1.826, leaveTangentWeight=1.0797, tangentMode=RCTM_User, time=0.0036, value=0.2824}
          {arriveTangentWeight=0.6623, interpMode=RCIM_Cubic, leaveTangentWeight=0.3962, tangentMode=RCTM_User, time=0.627, value=0.65}
          {arriveTangent=6.378, arriveTangentWeight=0.4894, interpMode=RCIM_Cubic, leaveTangent=6.378, tangentMode=RCTM_User, time=1, value=1.5}
         ]
        }}}
      CurveIndex=$Particles.Update Flash Star Age
      Scale Curve=$System.GlobalScale
     )
     Vector2D=$Particles.Initial.Flash Star Sprite Size
    )
    Particles.UpdateFlareCardColor:LinearColor=&MakeLinearColorFromVectorAndFloat(
     "Vector (RGB)"=&Multiply_Vector(
      A=&Multiply_Vector(
       A=&MakeVectorFromLinearColorRGB(Color=$Particles.Initial.FlareCardColor)
       B=&MakeVectorFromLinearColorRGB(Color=&ColorFromCurve(ColorCurve=DataInterface() {propertyValues {
           AlphaCurve {keys [<at 0 1 RCIM_Cubic>]}
           BlueCurve {keys [<at 0.0353 6>,<at 0.494 3>,<at 1 1>]}
           GreenCurve {keys [<at 0.0353 0.7328>,<at 0.494 0.5551>,<at 1 0>]}
           RedCurve {keys [<at 0.0353 0>,<at 0.494 0>,<at 1 0.0116>]}
          }}))
      )
      B=&VectorFromFloat(Value=&FN_FrontEndFloatModifier(Gameplay=1, Front End=0.7))
     )
     "Float (Alpha)"=&FloatFromCurve(
      FloatCurve=DataInterface() {propertyValues {Curve {keys [
          {arriveTangent=-1.9727, interpMode=RCIM_Cubic, time=0.1, value=1}
          {arriveTangent=-2.4337, interpMode=RCIM_Cubic, leaveTangent=-2.4337, tangentMode=RCTM_User, time=1, value=0}
         ]}}}
      Scale Curve=&MakeCustomFloatFromBool(Bool=$User.bIsFrontEndPreview, True Float=0.5, False Float=1)
     )
    )
   ) {isSetParameters=true}
   /SetVariables(
    Particles.Update Flash Star Color:LinearColor=&ColorFromCurve(
     ColorCurve=DataInterface() {propertyValues {
       AlphaCurve {keys [<at 0.7378 1>,<at 1 0>]}
       BlueCurve {keys [<at 0.003 10.0076>,<at 0.257 7>,<at 0.5238 1>,<at 0.9466 1.0039>]}
       GreenCurve {keys [<at 0.003 12>,<at 0.257 0.855>,<at 0.5238 0.1221>,<at 0.9466 0.0357>]}
       RedCurve {keys [<at 0.003 2.9883>,<at 0.257 0>,<at 0.5238 0>,<at 0.9466 0>]}
      }}
     CurveIndex=$Particles.Update Flash Star Age
    )
    Particles.Update Flash Star Sprize Size:Vector2f=&Multiply_Vector2D_ByFloat(
     Float=&FloatFromCurve(
      FloatCurve=DataInterface() {propertyValues {Curve {
         defaultValue=1
         keys [
          {arriveTangent=4.1785, interpMode=RCIM_Cubic, leaveTangent=4.1785, tangentMode=RCTM_User, time=0, value=0}
          {arriveTangent=1.1179, interpMode=RCIM_Cubic, leaveTangent=1.1179, tangentMode=RCTM_User, time=0.2042, value=0.6228}
          {arriveTangent=0.0224, interpMode=RCIM_Cubic, leaveTangent=0.0224, tangentMode=RCTM_User, time=1, value=1}
         ]
        }}}
      CurveIndex=$Particles.Update Flash Star Age
     )
     Vector2D=$Particles.Initial.Flash Star Sprite Size
    )
   ) {isSetParameters=true}
   /SetVariables(Particles.ColorDarkGlow:LinearColor=&MakeLinearColorFromVectorAndFloat(
     "Vector (RGB)"=&MakeVectorFromLinearColorRGB(Color=&Multiply_LinearColorByFloat(
       Linear Color=&ColorFromCurve(ColorCurve=DataInterface() {propertyValues {
          AlphaCurve {keys [<at 0.0334 1>,<at 1 0>]}
          BlueCurve {keys [<at 0 0.224>,<at 0.8496 0.0625>]}
          GreenCurve {keys [<at 0 0.0559>,<at 0.8496 0.0068>]}
          RedCurve {keys [<at 0 0>,<at 0.8496 0>]}
         }})
       Float=1
      ))
     "Float (Alpha)"=&FloatFromCurve(
      FloatCurve=DataInterface() {propertyValues {Curve {
         defaultValue=1
         keys [
          {arriveTangent=-1.1831, interpMode=RCIM_Cubic, leaveTangent=-1.1831, tangentMode=RCTM_User, time=0, value=1}
          {interpMode=RCIM_Cubic, tangentMode=RCTM_User, time=1, value=0}
         ]
        }}}
      Scale Curve=&FN_FrontEndFloatModifier(Gameplay=1, Front End=0.4)
     )
    )) {isSetParameters=true}
   &Drag(
    Drag=&MaskFloatBySpawnGroup(Masked Float=4, Passthrough Float=2, Spawn Group Mask=4)
    Use Linear Drag=true
    Use Rotational Drag=true
   )
   &SolveForcesAndVelocity(Clamp Velocity=false, Limit Acceleration=false)
   &FN_Light_Attributes(
    Color=&MultiplyLinearColor(Linear Color=$System.TOD Color, Scale Factor=&MakeVector4_FromLinearColor(LinearColor=<rgba 0 0.1221 1 1>))
    Create Color Attribute=true
    Create Exponent Attribute=true
    Create Position Attribute=true
    Create Radius Attribute=true
    Exponent=3
    Position=<0 5 0>
    Radius=90
    Scale Color=&FloatFromCurve(
     FloatCurve=DataInterface() {propertyValues {Curve {keys [
         {interpMode=RCIM_Cubic, tangentMode=RCTM_User, time=0, value=1}
         {arriveTangentWeight=0.3034, interpMode=RCIM_Cubic, tangentMode=RCTM_User, time=1, value=0}
        ]}}}
     Scale Curve=20
    )
   )
  ]
  rendererClasses [&NiagaraLightRendererProperties, &NiagaraSpriteRendererProperties]
  renderers [
   {rendererClass=&NiagaraLightRendererProperties, rendererIndex=0}
   {rendererClass=&NiagaraSpriteRendererProperties, rendererIndex=1}
   {rendererClass=&NiagaraSpriteRendererProperties, rendererIndex=2}
   {rendererClass=&NiagaraSpriteRendererProperties, rendererIndex=3}
   {rendererClass=&NiagaraSpriteRendererProperties, rendererIndex=4}
  ]
 }
 FN_PhysMatDebris {
  emitterUpdateScript [
   &EmitterState()
   &SpawnBurst_Instantaneous(
    Spawn Count=&RandomRangeInt(Minimum=5, Maximum=10)
    Use Spawn Probability=false
    UseLoopCountLimit=false
   )
  ]
  particleSpawnScript [
   /SetVariables(Particles.BiasedRandom=&FloatFromCurve(
     FloatCurve=DataInterface() {propertyValues {Curve {keys [
         <at 0.2 1 RCIM_Cubic>
         {arriveTangent=-3.3394, interpMode=RCIM_Cubic, leaveTangent=-3.3394, tangentMode=RCTM_User, time=0.5, value=0.5}
         <at 0.8 0 RCIM_Cubic>
        ]}}}
     CurveIndex=&RandomRangeFloat(Minimum=0, Maximum=1)
    )) {isSetParameters=true}
   &InitializeParticle(
    Color Mode=::NewEnumerator0
    Lifetime=&RandomRangeFloat(Minimum=0.5, Maximum=1.25)
    Mass=&Lerp_Float(A=0.6, B=1.2, Alpha=$Particles.BiasedRandom)
    Mass Mode=::NewEnumerator1
    Mesh Scale=&Multiply_VectorByFloat(
     Vector=&RandomRangeVector(Minimum=<0.9 0.4 0.5>, Maximum=<1 1 1>)
     Float=&Lerp_Float(A=0.7, B=1, Alpha=$Particles.BiasedRandom)
    )
    Mesh Scale Mode=::NewEnumerator4
    UsePositionOffset=false
    Write Material Random=true
   )
   &FN_PhysicalSurface(
    Dirt Mesh Count=2
    Glass Mesh Count=3
    Ice Mesh Count=2
    Metal Gravity=&RandomRangeFloat(Minimum=-350, Maximum=-1800)
    Metal Mesh Count=2
    Metal Velocity=&RandomRangeFloat(Minimum=80, Maximum=375)
    Metal Velocity Cone Angle=130
    Stone Gravity=&RandomRangeFloat(Minimum=-700, Maximum=-1250)
    Stone Mesh Count=2
    Stone Velocity=&RandomRangeFloat(Minimum=80, Maximum=550)
    Stone Velocity Cone Angle=120
    Wood Gravity=&RandomRangeFloat(Minimum=-450, Maximum=-900)
    Wood Mesh Count=2
    Wood Scale=1.5
    Wood Velocity=&RandomRangeFloat(Minimum=80, Maximum=600)
    Wood Velocity Cone Angle=130
   )
   &AddVelocityInCone(
    Cone Angle=$Output.FN_PhysicalSurface.SurfaceResponseConeAngle
    Module.Use Velocity Falloff On Cone Axis=false
    Velocity Strength=$Output.FN_PhysicalSurface.SurfaceResponseVelocity
   )
   &InitialMeshOrientation(Use Rotation Vector=false)
   &AddRotationalVelocity(Rotation Rate=&RandomVector(Vector Scale=&RandomRangeFloat(Minimum=1, Maximum=2)))
   &ApplyInitialForces()
   /SetVariables(Particles.GravityOverrideValue=$Output.FN_PhysicalSurface.GravityOverride) {isSetParameters=true}
  ]
  particleUpdateScript [
   &ParticleState()
   &GravityForce(Gravity=$Particles.GravityOverrideValue)
   &Collision(
    Advanced Aging Rate=1
    Age Colliding Particles=false
    CPU Collision Trace Channel=ECC_WorldStatic
    Collision Enabled=&Boolean_AndOperation(
     Bool1=&SetBoolByFloatComparison(A=$Particles.Initial.Lifetime, B=1)
     Bool2=&SetBoolByFloatComparison(A=$Particles.Age, B=0.12)
    )
    Control Roll On Collision=true
    Friction=1
    Friction During a Bounce=0.25
    Max CPU Trace Length=100
    Module.Control Rest State Via Surface Normal=false
    Module.EnableMaxCollisionCount=false
    Module.Randomize Collision Normal Vector=true
    Randomize Collision Normal=0.25
    ScaleRateOfRotation=false
   ) {enabled=false}
   &SolveForcesAndVelocity(Clamp Velocity=false, Limit Acceleration=false)
   &ScaleMeshSize(Scale Factor=&VectorFromFloat(Value=&FloatFromCurve(FloatCurve=DataInterface() {propertyValues {Curve {keys [<at 0.6 1 RCIM_Cubic>, {arriveTangent=-6.308, interpMode=RCIM_Cubic, leaveTangent=-6.308, tangentMode=RCTM_User, time=1, value=0}]}}})))
   &DynamicMaterialParameters(Module.Param0WriteEnabled=true, Module.Param1WriteEnabled=true, Module.Param2WriteEnabled=false, Module.Param3WriteEnabled=false)
  ]
  rendererClasses [&NiagaraSpriteRendererProperties, &NiagaraMeshRendererProperties]
  renderers [
   {rendererClass=&NiagaraSpriteRendererProperties}
   {rendererClass=&NiagaraMeshRendererProperties, rendererIndex=1}
   {rendererClass=&NiagaraMeshRendererProperties, rendererIndex=2}
   {rendererClass=&NiagaraMeshRendererProperties, rendererIndex=3}
   {rendererClass=&NiagaraMeshRendererProperties, rendererIndex=4}
   {rendererClass=&NiagaraMeshRendererProperties, rendererIndex=5}
   {rendererClass=&NiagaraMeshRendererProperties, rendererIndex=6}
   {rendererClass=&NiagaraMeshRendererProperties, rendererIndex=7}
   {rendererClass=&NiagaraMeshRendererProperties, rendererIndex=8}
  ]
 }
 Flash_Elements001 {
  emitterUpdateScript [
   &EmitterState()
   &SpawnBurst_Instantaneous(Use Spawn Probability=false, UseLoopCountLimit=false)
   &SpawnBurst_Instantaneous(Spawn Count=1, Spawn Group=1, Use Spawn Probability=false) {module=SpawnBurst_Instantaneous002}
   &SpawnBurst_Instantaneous(Spawn Count=1, Spawn Group=2, Use Spawn Probability=false) {module=SpawnBurst_Instantaneous003}
  ]
  particleSpawnScript [
   &InitializeParticle(
    Color=$System.TOD Color
    Lifetime=&MergeHelper_FloatSwitch(0=0.16, 1=0.22, 2=0.2, 3=?, Index=$Emitter.SpawnGroup, Num=::NewEnumerator1)
    Sprite Rotation Mode=::NewEnumerator2
    Sprite Size Mode=::NewEnumerator3
    Uniform Sprite Size=&Multiply_Float(
     A=&MergeHelper_FloatSwitch(0=75, 1=200, 2=125, 3=?, Index=$Emitter.SpawnGroup, Num=::NewEnumerator1)
     B=$System.GlobalScale
    )
    UsePositionOffset=false
    Write Material Random=true
   )
   &CameraOffset(Camera Offset Amount=&MergeHelper_FloatSwitch(0=20, 1=25, 2=10, 3=?, Index=$Emitter.SpawnGroup, Num=::NewEnumerator1))
   /SetVariables(Particles.DynamicParameterAlt2:Vector4f=&MakeVector4(X=1, Y=1, Z=0, W=0)) {isSetParameters=true}
   /SetVariables(
    Particles.SubImageIndex=&SelectFloatFromArray(
     Array Sampling Mode=::NewEnumerator0
     Direct Array Index=?
     Float Selection Array=DataInterface() {propertyValues {FloatData [1,2,3]}}
     Random Seed=0
    )
    Particles.SubImageIndexAlt:NiagaraFloat=&SelectFloatFromArray(
     Array Sampling Mode=::NewEnumerator0
     Direct Array Index=?
     Float Selection Array=DataInterface() {propertyValues {FloatData [0,1,2,3]}}
     Random Seed=0
    )
   ) {isSetParameters=true}
   /SetVariables(Particles.VisibilityTag=$Emitter.SpawnGroup) {isSetParameters=true}
  ]
  particleUpdateScript [
   &ParticleState()
   &ScaleSpriteSize(
    Scale Sprite Size Mode=::NewEnumerator0
    Uniform Scale Factor=&MergeHelper_FloatSwitch(
     0=&FloatFromCurve(FloatCurve=DataInterface() {propertyValues {Curve {keys [
          {arriveTangent=2.7399, interpMode=RCIM_Cubic, leaveTangent=2.7399, tangentMode=RCTM_User, time=0, value=0.1}
          {arriveTangent=0.2362, interpMode=RCIM_Cubic, leaveTangent=0.2362, tangentMode=RCTM_User, time=1, value=1}
         ]}}})
     1=&FloatFromCurve(FloatCurve=DataInterface() {propertyValues {Curve {keys [
          {arriveTangent=1.6131, interpMode=RCIM_Cubic, leaveTangent=1.6131, tangentMode=RCTM_User, time=0, value=0.7}
          {arriveTangent=0.0224, interpMode=RCIM_Cubic, leaveTangent=0.0224, tangentMode=RCTM_User, time=1, value=0}
         ]}}})
     2=&FloatFromCurve(FloatCurve=DataInterface() {propertyValues {Curve {keys [{arriveTangent=3.9086, interpMode=RCIM_Cubic, leaveTangent=3.9086, tangentMode=RCTM_User, time=0, value=0}, <at 1 1.4 RCIM_Cubic>]}}})
     3=?
     Index=$Particles.VisibilityTag
     Num=::NewEnumerator1
    )
   ) {module=ScaleSpriteSize001}
   &ScaleColor(
    Scale Alpha=&MergeHelper_FloatSwitch(
     0=&FloatFromCurve(FloatCurve=DataInterface() {propertyValues {Curve {keys [
          {arriveTangent=33.7585, interpMode=RCIM_Cubic, leaveTangent=33.7585, tangentMode=RCTM_User, time=0, value=0}
          {arriveTangent=0.3975, interpMode=RCIM_Cubic, leaveTangent=0.3975, tangentMode=RCTM_User, time=0.0555, value=1}
          {arriveTangent=-1.602, interpMode=RCIM_Cubic, leaveTangent=-1.602, tangentMode=RCTM_User, time=1, value=0}
         ]}}})
     1=&FloatFromCurve(FloatCurve=DataInterface() {propertyValues {Curve {
         defaultValue=1
         keys [
          {arriveTangent=-1.0114, interpMode=RCIM_Cubic, leaveTangent=-1.0114, tangentMode=RCTM_User, time=0.1, value=1}
          {arriveTangentWeight=0.7448, interpMode=RCIM_Cubic, tangentMode=RCTM_User, time=1, value=0}
         ]
        }}})
     2=&FloatFromCurve(
      FloatCurve=DataInterface() {propertyValues {Curve {keys [{arriveTangent=-0.2921, interpMode=RCIM_Cubic, leaveTangent=-0.2921, tangentMode=RCTM_User, time=0, value=7}, <at 1 0 RCIM_Cubic>]}}}
      Scale Curve=2
     )
     3=?
     Index=$Particles.VisibilityTag
     Num=::NewEnumerator1
    )
    Scale RGB=&MergeHelper_VectorSwitch(
     0=&Multiply_VectorByFloat(
      Vector=&MakeVectorFromLinearColorRGB(Color=&ColorFromCurve(ColorCurve=DataInterface() {propertyValues {
          AlphaCurve {keys [<at 0 1 RCIM_Cubic>]}
          BlueCurve {keys [<at 0 1 RCIM_Cubic>, {arriveTangent=-0.634, interpMode=RCIM_Cubic, leaveTangent=-0.634, time=0.7021, value=1}, <at 0.994 0.3698>]}
          GreenCurve {keys [<at 0 0.2423 RCIM_Cubic>, {arriveTangent=-0.1536, interpMode=RCIM_Cubic, leaveTangent=-0.1536, time=0.7021, value=0.2423}, <at 0.994 0.0896>]}
          RedCurve {keys [<at 0 0 RCIM_Cubic>,<at 0.7021 0 RCIM_Cubic>,<at 0.994 0>]}
         }}))
      Float=&FloatFromCurve(
       FloatCurve=DataInterface() {propertyValues {Curve {keys [{arriveTangent=-6.9854, interpMode=RCIM_Cubic, leaveTangent=-6.9854, tangentMode=RCTM_User, time=0.1, value=5}, <at 1 1 RCIM_Cubic>]}}}
       Scale Curve=2
      )
     )
     1=&MakeVectorFromLinearColorRGB(Color=&ColorFromCurve(ColorCurve=DataInterface() {propertyValues {
         AlphaCurve {keys [<at 0 1 RCIM_Cubic>]}
         BlueCurve {keys [<at 0 3>,<at 0.6934 1>,<at 1 0.5906>]}
         GreenCurve {keys [<at 0 0.7268>,<at 0.6934 0.2423>,<at 1 0.1431>]}
         RedCurve {keys [<at 0 0>,<at 0.6934 0>,<at 1 0>]}
        }}))
     2=<0 0 0>
     3=?
     Index=$Particles.VisibilityTag
     Num=::NewEnumerator1
    )
    ScaleA=true
    ScaleRGB=true
   )
   /SetVariables(
    Particles.InnerColor:LinearColor=&MultiplyLinearColor(Linear Color=<rgba 0 1.2125 10 1>, Scale Factor=&MakeVector4_FromLinearColor(LinearColor=$Particles.Color))
    Particles.InnerDynamicParameter:Vector4f=&MakeVector4_FromFloat(Float=&Lerp_Float(A=0, B=0.5, Alpha=$Particles.NormalizedAge))
    Particles.InnerSize:Vector2f=&Multiply_Vector2D_ByFloat(Float=0.7, Vector2D=$Particles.SpriteSize)
    Particles.MiddleColor:LinearColor=&MultiplyLinearColor(Linear Color=<rgba 0 0.6063 5 1>, Scale Factor=&MakeVector4_FromLinearColor(LinearColor=$Particles.Color))
    Particles.MiddleDynamicParameter:Vector4f=&MakeVector4_FromFloat(Float=&Lerp_Float(A=0.25, B=0.75, Alpha=$Particles.NormalizedAge))
    Particles.MiddleSize:Vector2f=&Multiply_Vector2D_ByFloat(Float=0.9, Vector2D=$Particles.SpriteSize)
    Particles.OuterColor:LinearColor=&MultiplyLinearColor(Linear Color=<rgba 0 0.2425 2 1>, Scale Factor=&MakeVector4_FromLinearColor(LinearColor=$Particles.Color))
    Particles.OuterDynamicParameter:Vector4f=&MakeVector4_FromFloat(Float=&Lerp_Float(A=0, B=1, Alpha=$Particles.NormalizedAge))
    Particles.OuterSize:Vector2f=&Multiply_Vector2D_ByFloat(Float=1.2, Vector2D=$Particles.SpriteSize)
    Particles.SubImageIndexAlt:NiagaraFloat=&RandomRangeFloat(Minimum=0, Maximum=3)
   ) {isSetParameters=true}
   /SetVariables(Particles.DynamicParameterAlt3:Vector4f=&MakeVector4(
     X=&FloatFromCurve(FloatCurve=DataInterface() {propertyValues {Curve {keys [<at 0 1 RCIM_Cubic>, {arriveTangent=-2.742, interpMode=RCIM_Cubic, leaveTangent=-2.742, tangentMode=RCTM_User, time=1, value=0}]}}})
     Y=&FloatFromCurve(FloatCurve=DataInterface() {propertyValues {Curve {keys [<at 0 0 RCIM_Cubic>, {arriveTangent=2.4833, interpMode=RCIM_Cubic, leaveTangent=2.4833, tangentMode=RCTM_User, time=1, value=1}]}}})
     Z=0
     W=0
    )) {isSetParameters=true}
   &FN_Light_Attributes(
    Color=<rgba 0 0.1213 1 1>
    Create Color Attribute=true
    Create Exponent Attribute=true
    Create Position Attribute=true
    Create Radius Attribute=true
    Enable Visibility Tag=true
    Exponent=5
    Position=<15 0 0>
    Radius=90
    Scale Color=&FloatFromCurve(
     FloatCurve=DataInterface() {propertyValues {Curve {keys [
         {interpMode=RCIM_Cubic, tangentMode=RCTM_User, time=0, value=1}
         {arriveTangentWeight=0.3413, interpMode=RCIM_Cubic, tangentMode=RCTM_User, time=1, value=0}
        ]}}}
     Scale Curve=10
    )
    VisibilityTag=2
   )
  ]
  rendererClasses [&NiagaraSpriteRendererProperties]
  renderers [
   {rendererClass=&NiagaraSpriteRendererProperties}
   {rendererClass=&NiagaraSpriteRendererProperties, rendererIndex=1}
   {rendererClass=&NiagaraSpriteRendererProperties, rendererIndex=2}
   {rendererClass=&NiagaraSpriteRendererProperties, rendererIndex=3}
  ]
 }
 Smoke {
  emitterUpdateScript [
   &EmitterState()
   &SpawnBurst_Instantaneous(Spawn Count=1, Use Spawn Probability=false)
  ]
  particleSpawnScript [
   &CameraOffset(Camera Offset Amount=30)
   &DynamicMaterialParameters(
    Index 0 Param 2=&UniformAOrBFloat(A=0, Evaluation Type=::NewEnumerator0, Recalculate Random Each Loop=true, Override Seed=false)
    Module.Index0Float0Write=true
    Module.Index0Float1Write=true
    Module.Index0Float2Write=true
    Module.Index0Float3Write=true
   ) {module=DynamicMaterialParameters001}
   &InitializeParticle(
    Color=$System.TOD
    Lifetime=0.52
    Sprite Rotation Mode=::NewEnumerator2
    Sprite Size Mode=::NewEnumerator3
    Sprite UV Mode=::NewEnumerator3
    Uniform Sprite Size=&Multiply_Float(A=95, B=$System.GlobalScale)
    UsePositionOffset=false
    Write Material Random=true
   ) {module=InitializeParticle001}
  ]
  particleUpdateScript [
   &ParticleState()
   &ScaleSpriteSize(Uniform Curve Sprite Scale=DataInterface() {propertyValues {Curve {keys [
        {arriveTangent=0.6997, interpMode=RCIM_Cubic, leaveTangent=0.6997, tangentMode=RCTM_User, time=0, value=0.7}
        {arriveTangentWeight=0.3042, interpMode=RCIM_Cubic, leaveTangentWeight=0.0845, tangentMode=RCTM_User, time=0.7827, value=0.9719}
        {arriveTangent=2.7371, interpMode=RCIM_Cubic, leaveTangent=2.7371, tangentMode=RCTM_User, time=1, value=1.2153}
       ]}}})
   &DynamicMaterialParameters(
    Index 0 Param 1=&FloatFromCurve(FloatCurve=DataInterface() {propertyValues {Curve {keys [
         {arriveTangent=1.9981, interpMode=RCIM_Cubic, leaveTangent=1.9981, tangentMode=RCTM_User, time=0, value=0}
         {arriveTangent=0.6664, interpMode=RCIM_Cubic, leaveTangent=0.6664, tangentMode=RCTM_User, time=0.4875, value=0.7253}
         {arriveTangentWeight=0.1717, interpMode=RCIM_Cubic, tangentMode=RCTM_User, time=1, value=1}
        ]}}})
    Index 0 Param 2=1
    Module.Param0WriteEnabled=true
    Module.Param1WriteEnabled=true
    Module.Param2WriteEnabled=true
    Module.Param3WriteEnabled=true
   )
   &ScaleColor(
    Scale Alpha=&FloatFromCurve(
     FloatCurve=DataInterface() {propertyValues {Curve {keys [<at 0.3 1 RCIM_Cubic>, {arriveTangent=-2.4767, interpMode=RCIM_Cubic, leaveTangent=-2.4767, tangentMode=RCTM_User, time=1, value=0}]}}}
     Scale Curve=0.65
    )
    Scale RGB=&MakeVectorFromLinearColorRGB(Color=<rgba 1 1 1 1>)
    ScaleA=true
    ScaleRGB=true
   )
  ]
  rendererClasses [&NiagaraSpriteRendererProperties]
  renderers [{rendererClass=&NiagaraSpriteRendererProperties}]
 }
 Sparks_Packed {
  emitterUpdateScript [
   &EmitterState()
   &SpawnBurst_Instantaneous(
    Spawn Count=&RandomRangeInt(Minimum=15, Maximum=20)
    Use Spawn Probability=false
    UseLoopCountLimit=false
   )
   &SpawnBurst_Instantaneous(
    Spawn Count=&RandomRangeInt(Minimum=3, Maximum=7)
    Spawn Group=1
    Use Spawn Probability=false
    UseLoopCountLimit=false
   ) {module=SpawnBurst_Instantaneous001}
  ]
  particleSpawnScript [
   /SetVariables(Particles.VisibilityTag=$Emitter.SpawnGroup) {isSetParameters=true}
   &InitializeParticle(
    Color=<rgba 0 0.2425 2 1>
    Lifetime=&MergeHelper_FloatSwitch(
     0=&RandomRangeFloat(Minimum=0.05, Maximum=0.45)
     1=&RandomRangeFloat(Minimum=0.35, Maximum=0.65)
     2=?
     3=?
     Index=$Particles.VisibilityTag
     Num=::NewEnumerator0
    )
    Sprite Size Mode=::NewEnumerator3
    Uniform Sprite Size=&Multiply_Float(
     A=&MakeCustomFloatFromBool(Bool=$User.bIsFrontEnd, True Float=0.75, False Float=1)
     B=&RandomRangeFloat(Minimum=3, Maximum=5)
    )
    UsePositionOffset=false
    Write Material Random=true
   )
   &ShapeLocation(
    Hemisphere Distribution=&FN_FrontEndVector2Modifier(Front End:Vector2f=<1 1>, Gameplay:Vector2f=<1 0.5>)
    MaskSpawnGroup=false
    Offset=$System.FEOffset
    Offset Coordinate Space=::Local
    Offset Mode=::NewEnumerator0
    Override Seed=false
    Rotation Angle=90
    Rotation Angle Type:ENiagara_AngleInput=::NewEnumerator0
    Rotation Mode=::NewEnumerator1
    Sphere Radius=&MergeHelper_FloatSwitch(0=9, 1=25, 2=?, 3=?, Index=$Particles.VisibilityTag, Num=::NewEnumerator0)
    Sphere Surface Distribution:NiagaraFloat=&MergeHelper_FloatSwitch(0=0, 1=0.7, 2=?, 3=?, Index=$Particles.VisibilityTag, Num=::NewEnumerator0)
    UseOwnerScale=true
   )
   &AddVelocity(
    Module.Use Falloff Distance=false
    Origin Offset=$System.FEOffset
    Override Seed=false
    Velocity Mode=::NewEnumerator1
    Velocity Speed=&MergeHelper_FloatSwitch(
     0=&FN_FrontEndFloatModifier(
      Gameplay=&RandomRangeFloat(Minimum=222, Maximum=666)
      Front End=&RandomRangeFloat(Minimum=150, Maximum=375)
     )
     1=&RandomRangeFloat(Minimum=15, Maximum=35)
     2=?
     3=?
     Index=$Particles.VisibilityTag
     Num=::NewEnumerator0
    )
   )
  ]
  particleUpdateScript [
   &ParticleState()
   &GravityForce(Gravity=&MergeHelper_VectorSwitch(0=<0 0 -666>, 1=<0 0 0>, 2=?, 3=?, Index=$Particles.VisibilityTag, Num=::NewEnumerator0))
   &Drag(Drag=&UniformRangedFloat(Minimum=2, Maximum=6), Use Linear Drag=true, Use Rotational Drag=true)
   &SolveForcesAndVelocity(Clamp Velocity=false, Limit Acceleration=false)
   &ScaleSpriteSize(Uniform Curve Sprite Scale=DataInterface() {propertyValues {Curve {keys [
        {arriveTangent=-6.725, arriveTangentWeight=0.2167, interpMode=RCIM_Cubic, leaveTangent=-6.725, tangentMode=RCTM_User, time=0.65, value=1}
        <at 1 0 RCIM_Cubic>
       ]}}})
   &ScaleSpriteSizeBySpeed(Max Scale Factor=<0.5 10>, Min Scale Factor=<0.5 1>)
   &ScaleColor(
    Scale RGB=&VectorFromFloat(Value=&FloatFromCurve(
      FloatCurve=DataInterface() {propertyValues {Curve {keys [
          {interpMode=RCIM_Cubic, tangentMode=RCTM_User, time=0, value=1}
          <at 0.3 0.2>
          {interpMode=RCIM_Cubic, tangentMode=RCTM_User, time=1, value=0}
         ]}}}
      Scale Curve=10
     ))
    ScaleA=true
    ScaleRGB=true
   )
   /SetVariables(
    Particles.GlowColor=&MakeLinearColorFromVectorAndFloat(
     "Vector (RGB)"=&MakeVectorFromLinearColorRGB(Color=<rgba 0 0.1213 1 1>)
     "Float (Alpha)"=&FloatFromCurve(FloatCurve=DataInterface() {propertyValues {Curve {keys [<at 0.5 1>,<at 1 0>]}}}, Scale Curve=0.1)
    )
    Particles.GlowSize=&Multiply_Vector2D(A=$Particles.SpriteSize, B=<8 2>)
   ) {isSetParameters=true}
   &FN_Sprite_Attributes(
    Camera Offset=10
    Color=<rgba 0 0.1213 1 1>
    Create Camera Offset Attribute=true
    Create Color Attribute=true
    Create SpriteSize Attribute=true
    Module Output:ENiagara_AttributeSamplingApplyOutput=::NewEnumerator1
    SpriteSize=&Multiply_Vector2D(
     A=&Vector2DFromFloat(Value=&Multiply_Float(
       A=&RandomRangeFloat(Minimum=30, Maximum=41.25)
       B=&FN_FrontEndFloatModifier(Gameplay=1, Front End=0.8)
      ))
     B=&Multiply_Vector2D_ByFloat(
      Float=&FloatFromCurve(FloatCurve=DataInterface() {propertyValues {Curve {keys [
           {arriveTangent=1.9457, interpMode=RCIM_Cubic, leaveTangent=1.9457, tangentMode=RCTM_User, time=0, value=0.35}
           {arriveTangent=-3.1633, interpMode=RCIM_Cubic, leaveTangent=-3.1633, tangentMode=RCTM_User, time=0.5, value=1}
           {tangentMode=RCTM_User, time=1, value=0}
          ]}}})
      Vector2D=&Vector2DFromCurve(
       Vector2Curve=DataInterface() {propertyValues {
         XCurve {keys [
           {arriveTangent=-1.8625, interpMode=RCIM_Cubic, leaveTangent=-1.8625, tangentMode=RCTM_User, time=0, value=1}
           {tangentMode=RCTM_User, time=1, value=0.35}
          ]}
         YCurve {keys [
           {arriveTangent=3.1672, interpMode=RCIM_Cubic, leaveTangent=3.1672, tangentMode=RCTM_User, time=0, value=1}
           {arriveTangent=-3.3667, interpMode=RCIM_Cubic, leaveTangent=-3.3667, tangentMode=RCTM_User, time=0.3, value=1.1621}
           {arriveTangent=-0.924, interpMode=RCIM_Cubic, leaveTangent=-0.924, tangentMode=RCTM_User, time=0.5, value=0.7}
           {tangentMode=RCTM_User, time=1, value=0.5}
          ]}
        }}
       ScaleCurve=<1 1>
      )
     )
    )
   ) {module=FN_Sprite_Attributes001}
   &FN_Sprite_Attributes(
    Color=&MakeLinearColorFromVectorAndFloat("Vector (RGB)"=&MakeVectorFromLinearColorRGB(Color=<rgba 0 0.2425 2 1>), "Float (Alpha)"=0.15)
    Create Color Attribute=true
    Module Output:ENiagara_AttributeSamplingApplyOutput=::NewEnumerator1
   )
  ]
  rendererClasses [&NiagaraSpriteRendererProperties]
  renderers [
   {rendererClass=&NiagaraSpriteRendererProperties}
   {rendererClass=&NiagaraSpriteRendererProperties, rendererIndex=1}
   {rendererClass=&NiagaraSpriteRendererProperties, rendererIndex=2}
   {rendererClass=&NiagaraSpriteRendererProperties, rendererIndex=3}
  ]
 }
 directional_Spikes {
  emitterSpawnScript [/SetVariables(Emitter.SpawnCount=10) {isSetParameters=true}]
  emitterUpdateScript [
   &EmitterState()
   &SpawnBurst_Instantaneous(Spawn Count=$Emitter.SpawnCount, Use Spawn Probability=false, UseLoopCountLimit=false)
  ]
  particleSpawnScript [
   /SetVariables(Particles.Variation:NiagaraFloat=&FloatFromCurve(
     FloatCurve=DataInterface() {propertyValues {Curve {keys [{tangentMode=RCTM_User, time=0, value=0}, {tangentMode=RCTM_User, time=1, value=1}]}}}
     CurveIndex=&Divide_Float(
      A=&MakeFloatFromInt(Int=&ReturnExecIndex())
      B=&MakeFloatFromInt(Int=&Add_Integer(A=$Emitter.SpawnCount, B=-1))
     )
    )) {isSetParameters=true}
   &InitializeParticle(
    Color=<rgba 0 0.117 0.9647 0.7>
    Lifetime=&Lerp_Float(A=0.27, B=0.2, Alpha=$Particles.Variation)
    Sprite Size Mode=::NewEnumerator3
    Uniform Sprite Size=&Lerp_Float(A=25, B=50, Alpha=$Particles.Variation)
    UsePositionOffset=false
    Write Material Random=true
   )
   &ShapeLocation(
    MaskSpawnGroup=false
    Sphere Distribution=::NewEnumerator3
    Sphere Radius=2
    Uniform Spiral Amount:NiagaraFloat=&RandomRangeFloat(Minimum=0.7, Maximum=1)
    UseOwnerScale=true
   )
   &AddVelocity(
    Module.Use Falloff Distance=false
    Override Seed=false
    Velocity Mode=::NewEnumerator1
    Velocity Speed=&Multiply_Float(A=&RandomRangeFloat(Minimum=250, Maximum=400), B=$System.GlobalScale)
   )
   &AddVelocity(
    Velocity=&RandomRangeVector(Minimum=<100 0 0>, Maximum=<150 0 0>)
    Velocity Speed Scale=&FN_FrontEndFloatModifier(Gameplay=1, Front End=0)
   ) {module=AddVelocity001}
   /SetVariables(Particles.LineSubUV:NiagaraFloat=&MakeFloatFromInt(Int=&RandomRangeInt(Minimum=0, Maximum=3))) {isSetParameters=true}
   &CameraOffset(Camera Offset Amount=2)
  ]
  particleUpdateScript [
   &Drag(Drag=5, Use Linear Drag=true, Use Rotational Drag=false)
   &SolveForcesAndVelocity(Clamp Velocity=false, Limit Acceleration=false)
   &ParticleState()
   &ScaleSpriteSize(
    Uniform Curve Scale=$System.GlobalScale
    Uniform Curve Sprite Scale=DataInterface() {propertyValues {Curve {keys [
        {arriveTangent=4.2986, interpMode=RCIM_Cubic, leaveTangent=4.2986, tangentMode=RCTM_User, time=0, value=0}
        {tangentMode=RCTM_User, time=0.5, value=1}
       ]}}}
   )
   &DynamicMaterialParameters(
    Index 0 Param 1=&FloatFromCurve(FloatCurve=DataInterface() {propertyValues {Curve {keys [
         {arriveTangent=0.8115, interpMode=RCIM_Cubic, leaveTangent=0.8115, tangentMode=RCTM_User, time=0.25, value=0.2}
         {arriveTangent=0.7921, interpMode=RCIM_Cubic, leaveTangent=0.7921, tangentMode=RCTM_User, time=1, value=1}
        ]}}})
    Index 0 Param 2=&SelectFloatFromArray(
     Array Sampling Mode=::NewEnumerator0
     Direct Array Index=?
     Float Selection Array=DataInterface() {propertyValues {FloatData [0,3]}}
     Random Seed=0
    )
    Index 1 Param 2=1
    Index 1 Param 3=1
    Index 1 Param 4=&FloatFromCurve(FloatCurve=DataInterface() {propertyValues {Curve {keys [<at 0.5 1 RCIM_Cubic>, {arriveTangent=-3.1839, tangentMode=RCTM_User, time=1, value=0.35}]}}})
    Module.Index0Float0Write=true
    Module.Index0Float1Write=true
    Module.Index0Float2Write=true
    Module.Index0Float3Write=true
    Module.Index1Float0Write=true
    Module.Index1Float1Write=true
    Module.Index1Float2Write=true
    Module.Index1Float3Write=true
    Write Parameter Index 1=::NewEnumerator1
   )
   &ScaleColor(
    Scale Alpha=&FloatFromCurve(FloatCurve=DataInterface() {propertyValues {Curve {keys [
         {arriveTangent=7.213, interpMode=RCIM_Cubic, leaveTangent=7.213, tangentMode=RCTM_User, time=0, value=0.6}
         {arriveTangentWeight=0.0568, interpMode=RCIM_Cubic, leaveTangentWeight=0.1099, tangentMode=RCTM_User, time=0.1023, value=1}
         {interpMode=RCIM_Cubic, leaveTangentWeight=0.3967, tangentMode=RCTM_User, time=0.3, value=1}
         {arriveTangent=-1.961, interpMode=RCIM_Cubic, leaveTangent=-1.961, tangentMode=RCTM_User, time=1, value=0}
        ]}}})
    Scale RGB=&Multiply_Vector(
     A=&MakeVectorFromLinearColorRGB(Color=&ColorFromCurve(ColorCurve=DataInterface() {propertyValues {
         AlphaCurve {keys [<at 0 1 RCIM_Cubic>]}
         BlueCurve {keys [<at 0.0348 1.5>,<at 0.6 0.5104>,<at 1 0.5417>]}
         GreenCurve {keys [<at 0.0348 1.2418>,<at 0.6 0.3676>,<at 1 0.3583>]}
         RedCurve {keys [<at 0.0348 1.0411>,<at 0.6 0.2224>,<at 1 0.0987>]}
        }}))
     B=&MakeVectorFromLinearColorRGB(Color=<rgba 1 1 1 1>)
    )
    ScaleA=true
    ScaleRGB=true
   )
   /SetVariables(
    Transient.DistanceScale:NiagaraFloat=&FloatFromCurve(
     FloatCurve=DataInterface() {propertyValues {Curve {keys [<at 0 1 RCIM_Cubic>, {arriveTangent=-2.2435, tangentMode=RCTM_User, time=1, value=0.1}]}}}
     CurveIndex=&Multiply_Float(A=$Particles.NormalizedAge, B=&RandomRangeFloat(Minimum=0.75, Maximum=1))
     Scale Curve=&Multiply_Float(
      A=&Multiply_Float(A=2, B=$System.GlobalScale)
      B=&VectorLength(Vector=&Subtract_Vector(
        A=Linked() {linkedVariable {name=Particles.Position, type=NiagaraPosition}}
        B=Linked() {linkedVariable {name=Particles.Initial.Position, type=NiagaraPosition}}
       ))
     )
    )
    Transient.WidthScale:NiagaraFloat=&FloatFromCurve(
     FloatCurve=DataInterface() {propertyValues {Curve {keys [<at 0 0.3>,<at 1 0.2>]}}}
     Scale Curve=&Multiply_Float(A=&MakeFloatFromVector2d(Vector 2D=$Particles.Initial.SpriteSize), B=$System.GlobalScale)
    )
   ) {isSetParameters=true}
   /SetVariables(
    Particles.LineColor:LinearColor=&MultiplyLinearColor(Linear Color=$Particles.Color, Scale Factor=<1 1 1 1>)
    Particles.LineScale:Vector2f=&MakeVector2D(X=$Transient.WidthScale, Y=$Transient.DistanceScale)
   ) {isSetParameters=true}
   &SpriteFacingAndAlignment(
    Sprite Alignment=&Multiply_VectorByFloat(Vector=$Particles.Velocity, Float=-1)
    WriteAlignment=true
    WriteFacing=false
   )
  ]
  rendererClasses [&NiagaraSpriteRendererProperties]
  renderers [{rendererClass=&NiagaraSpriteRendererProperties}]
 }
}
summary {
 systemName=NS_Pickaxe_Guitar_SaltyBrim_Impact
 emitters [
  {
   emitterName=FN_PhysMatDebris
   rendererClasses [&NiagaraSpriteRendererProperties, &NiagaraMeshRendererProperties]
  }
  {emitterName=directional_Spikes, rendererClasses [&NiagaraSpriteRendererProperties]}
  {emitterName=Smoke, rendererClasses [&NiagaraSpriteRendererProperties]}
  {emitterName=Sparks_Packed, rendererClasses [&NiagaraSpriteRendererProperties]}
  {
   emitterName=Core_Impact_Burst001
   rendererClasses [&NiagaraLightRendererProperties, &NiagaraSpriteRendererProperties]
  }
  {emitterName=Flash_Elements001, rendererClasses [&NiagaraSpriteRendererProperties]}
 ]
 userVariables [
  User.bIsFrontEnd {type=NiagaraBool}
  User.bIsFrontEndPreview {type=NiagaraBool}
 ]
}
